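# Physics asset for the LycheeNickel backpack cosmetic, giving its pizza and dangling cheese bones capsule bodies and joint limits for secondary motion.
# A large pack capsule anchors base bodies for the pizza, upper cheese and two lower cheese drips.
# Dynamic bodies hang from these bases, including a two-link right strand and a tapering three-link long strand.
# Every joint has limited swing and zero twist limit, with the long strand swinging widest.
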PhysicsAsset /BRCosmetics/Accessories/FORT_Backpacks/Backpack_M_MED_LycheeNickel/Meshes/M_MED_LycheeNickel_Pack_Physics
bodies {
 Base_pizza [{shapeType=Capsule, length=0.505, radius=0.505, shapeName=Base_pizza_capsule, rotation=<rot 0 -90 0>}]
 Base_upperTap [{shapeType=Capsule, length=0.505, radius=0.505, shapeName=Base_upperTap_capsule, rotation=<rot 0 -90 0>}]
 base_LoweCheese_L [{shapeType=Capsule, length=0.505, radius=0.505, shapeName=base_LoweCheese_L_capsule, rotation=<rot 0 -90 0>}]
 base_LoweCheese_R [{shapeType=Capsule, length=0.505, radius=0.505, shapeName=base_LoweCheese_R_capsule, rotation=<rot 0 -90 0>}]
 dyn_LoweCheese_L [{shapeType=Capsule, length=0.505, radius=0.505, shapeName=dyn_LoweCheese_L_capsule, center=<2 0 0>, rotation=<rot 90 0 0>}]
 dyn_LoweCheese_R [{shapeType=Capsule, length=0.505, radius=0.505, shapeName=dyn_LoweCheese_R_capsule, center=<2 0 0>, rotation=<rot 90 0 0>}]
 dyn_MidChesse_R_01 [{shapeType=Capsule, length=0.505, radius=0.505, shapeName=dyn_MidChesse_R_01_capsule, center=<2 0 0>, rotation=<rot 90 0 0>}]
 dyn_cheese_L [{shapeType=Capsule, length=0.7097, radius=0.5652, shapeName=dyn_cheese_L_capsule, center=<2 0 0>, rotation=<rot 90 0 0>}]
 dyn_cheese_R_01 [{shapeType=Capsule, length=0.505, radius=0.505, shapeName=dyn_cheese_R_01_capsule, center=<2 0 0>, rotation=<rot 90 0 0>}]
 dyn_cheese_R_02 [{shapeType=Capsule, length=0.9325, radius=0.6359, shapeName=dyn_cheese_R_02_capsule, center=<2 0 0>, rotation=<rot 90 0 0>}]
 dyn_longCheese_01 [{shapeType=Capsule, length=1.9424, radius=0.9221, shapeName=dyn_longCheese_01_capsule, center=<2 0 0>, rotation=<rot 90 0 0>}]
 dyn_longCheese_02 [{shapeType=Capsule, length=1.5718, radius=0.7544, shapeName=dyn_longCheese_02_capsule, center=<2 0 0>, rotation=<rot 90 0 0>}]
 dyn_longCheese_03 [{shapeType=Capsule, length=1.1325, radius=0.5528, shapeName=dyn_longCheese_03_capsule, center=<2 0 0>, rotation=<rot 90 0 0>}]
 dyn_pizza [{shapeType=Capsule, length=2.0849, radius=1.4526, shapeName=dyn_pizza_capsule, center=<2 0 0>, rotation=<rot 90 0 0>}]
 dyn_smallChesse_R [{shapeType=Capsule, length=0.505, radius=0.505, shapeName=dyn_smallChesse_R_capsule, center=<2 0 0>, rotation=<rot 90 0 0>}]
 pack [{shapeType=Capsule, length=15.2165, radius=11.8887, center=<-0.3516 0 0>, rotation=<rot 90 0 0>}]
}
constraints [
 {bone1Name=dyn_MidChesse_R_01, bone2Name=Base_upperTap, swing1LimitDegrees=45, swing2LimitDegrees=45, twistLimitDegrees=0}
 {bone1Name=dyn_cheese_L, bone2Name=Base_upperTap, swing1LimitDegrees=45, swing2LimitDegrees=45, twistLimitDegrees=0}
 {bone1Name=dyn_cheese_R_01, bone2Name=Base_upperTap, swing1LimitDegrees=45, swing2LimitDegrees=45, twistLimitDegrees=0}
 {bone1Name=dyn_cheese_R_02, bone2Name=dyn_cheese_R_01, swing1LimitDegrees=45, swing2LimitDegrees=45, twistLimitDegrees=0}
 {bone1Name=dyn_longCheese_01, bone2Name=Base_upperTap, swing1LimitDegrees=70, swing2LimitDegrees=70, twistLimitDegrees=0}
 {bone1Name=dyn_longCheese_02, bone2Name=dyn_longCheese_01, swing1LimitDegrees=70, swing2LimitDegrees=70, twistLimitDegrees=0}
 {bone1Name=dyn_longCheese_03, bone2Name=dyn_longCheese_02, swing1LimitDegrees=70, swing2LimitDegrees=70, twistLimitDegrees=0}
 {bone1Name=dyn_smallChesse_R, bone2Name=Base_upperTap, swing1LimitDegrees=45, swing2LimitDegrees=45, twistLimitDegrees=0}
 {bone1Name=dyn_LoweCheese_L, bone2Name=base_LoweCheese_L, swing1LimitDegrees=45, swing2LimitDegrees=45, twistLimitDegrees=0}
 {bone1Name=dyn_LoweCheese_R, bone2Name=base_LoweCheese_R, swing1LimitDegrees=45, swing2LimitDegrees=45, twistLimitDegrees=0}
 {bone1Name=base_LoweCheese_R, bone2Name=pack, swing1LimitDegrees=60, swing2LimitDegrees=60, twistLimitDegrees=0}
 {bone1Name=base_LoweCheese_L, bone2Name=pack, swing1LimitDegrees=60, swing2LimitDegrees=60, twistLimitDegrees=0}
 {bone1Name=Base_upperTap, bone2Name=pack, swing1LimitDegrees=45, swing2LimitDegrees=45, twistLimitDegrees=0}
 {bone1Name=dyn_pizza, bone2Name=Base_pizza, swing1LimitDegrees=45, swing2LimitDegrees=45, twistLimitDegrees=0}
 {bone1Name=Base_pizza, bone2Name=pack, swing1LimitDegrees=45, swing2LimitDegrees=45, twistLimitDegrees=0}
]
